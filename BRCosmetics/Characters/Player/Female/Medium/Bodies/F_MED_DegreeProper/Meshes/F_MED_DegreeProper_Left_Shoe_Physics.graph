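# Physics asset for the left shoe of the female medium DegreeProper player body cosmetic, covering two dangling lace chains.
# Each lace runs from a sphere base through two capsule segments, joined by tight-twist, limited-swing constraints.
# Both bases attach to the left calf twist bone with free rotation.
# Capsules on that calf bone and the left foot give the leg's shape, with the foot body left unconstrained.
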PhysicsAsset /BRCosmetics/Characters/Player/Female/Medium/Bodies/F_MED_DegreeProper/Meshes/F_MED_DegreeProper_Left_Shoe_Physics
bodies {
 calf_twist_01_l [{
   shapeType=Capsule
   length=9.0104
   radius=3.725
   shapeName=calf_twist_01_l_capsule
   center=<-1.3268 0.8634 0.5295>
   rotation=<rot 57.9058 91.7263 -86.5738>
  }]
 dyn_lace_left_leg_l_01 [{
   shapeType=Capsule
   length=3.4605
   radius=0.9557
   shapeName=dyn_lace_left_leg_l_01_capsule
   center=<4.4246 0.2784 0.2596>
   rotation=<rot 0 -96.2862 -77.0657>
  }]
 dyn_lace_left_leg_l_02 [{
   shapeType=Capsule
   length=2.0224
   radius=0.7257
   shapeName=dyn_lace_left_leg_l_02_capsule
   center=<4.3047 -1.0285 -0.4738>
   rotation=<rot 0 87.9047 -85.5842>
  }]
 dyn_lace_left_leg_l_base [{shapeType=Sphere, radius=0.505, shapeName=dyn_lace_left_leg_l_base_sphere}]
 dyn_lace_left_leg_r_01 [{
   shapeType=Capsule
   length=2.8917
   radius=0.6703
   shapeName=dyn_lace_left_leg_r_01_capsule
   center=<4.8311 -0.0978 -0.3296>
   rotation=<rot 0 94.7649 -74.866>
  }]
 dyn_lace_left_leg_r_02 [{
   shapeType=Capsule
   length=1.9814
   radius=0.6582
   shapeName=dyn_lace_left_leg_r_02_capsule
   center=<4.5731 0.0176 0.1105>
   rotation=<rot 0 -93.0742 -84.9321>
  }]
 dyn_lace_left_leg_r_base [{shapeType=Sphere, radius=0.505, shapeName=dyn_lace_left_leg_r_base_sphere}]
 foot_l [{shapeType=Capsule, length=15.9887, radius=6.0609, center=<-7.174 8.383 -0.606>, rotation=<rot -0.051 10.51 -87.7823>}]
}
constraints [
 {bone1Name=dyn_lace_left_leg_l_02, bone2Name=dyn_lace_left_leg_l_01, swing1LimitDegrees=45, swing2LimitDegrees=20, twistLimitDegrees=1}
 {bone1Name=dyn_lace_left_leg_r_02, bone2Name=dyn_lace_left_leg_r_01, swing1LimitDegrees=45, swing2LimitDegrees=20, twistLimitDegrees=1}
 {bone1Name=dyn_lace_left_leg_l_01, bone2Name=dyn_lace_left_leg_l_base, swing1LimitDegrees=45, swing2LimitDegrees=20, twistLimitDegrees=1}
 {bone1Name=dyn_lace_left_leg_r_01, bone2Name=dyn_lace_left_leg_r_base, swing1LimitDegrees=45, swing2LimitDegrees=20, twistLimitDegrees=1}
 {
  bone1Name=dyn_lace_left_leg_r_base
  bone2Name=calf_twist_01_l
  swing1LimitDegrees=45
  swing1Motion=Free
  swing2LimitDegrees=45
  swing2Motion=Free
  twistLimitDegrees=45
  twistMotion=Free
 }
 {
  bone1Name=dyn_lace_left_leg_l_base
  bone2Name=calf_twist_01_l
  swing1LimitDegrees=45
  swing1Motion=Free
  swing2LimitDegrees=45
  swing2Motion=Free
  twistLimitDegrees=45
  twistMotion=Free
 }
]
