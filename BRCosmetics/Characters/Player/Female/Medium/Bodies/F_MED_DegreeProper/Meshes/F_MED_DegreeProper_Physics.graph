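# Physics asset for the F_MED_DegreeProper female medium player body cosmetic, set up around the motion of its scarf.
# Capsule bodies cover the upper spine, the neck and three dyn_scarf bones.
# The scarf root constrains to neck_01 and branches into left and right pieces, each joint having narrow swing and very tight twist limits.
# Neck and spine joints leave swing and twist free, chaining neck_01 through spine_05 to spine_04.
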PhysicsAsset /BRCosmetics/Characters/Player/Female/Medium/Bodies/F_MED_DegreeProper/Meshes/F_MED_DegreeProper_Physics
bodies {
 dyn_scarf [{
   shapeType=Capsule
   length=1.4469
   radius=1.9245
   shapeName=dyn_scarf_capsule
   center=<5.1399 0.7382 -0.3148>
   rotation=<rot 0 86.8688 -93.7734>
  }]
 dyn_scarf_l [{shapeType=Capsule, length=0, radius=2, shapeName=dyn_scarf_l_capsule, center=<3.6984 0.3031 -0.3279>, rotation=<rot 0.0034 87.579 -94.0262>}]
 dyn_scarf_r [{shapeType=Capsule, length=0, radius=2, shapeName=dyn_scarf_r_capsule, center=<2.1996 0.2177 -0.3419>, rotation=<rot -1.1298 -90 -80.9424>}]
 neck_01 [{shapeType=Capsule, length=0.505, radius=0.505, shapeName=neck_01_capsule, rotation=<rot 0 -90 0>}]
 spine_04 [{
   shapeType=Capsule
   length=13.5184
   radius=9.2716
   shapeName=spine_04_capsule
   center=<2.0016 -0.6043 0.2616>
   rotation=<rot -3.1713 -90 2.0227>
  }]
 spine_05 [{shapeType=Capsule, length=9.7776, radius=8.9184, shapeName=spine_05_capsule, center=<0.0189 -2.5895 0>}]
}
constraints [
 {
  bone1Name=neck_01
  bone2Name=spine_05
  swing1LimitDegrees=45
  swing1Motion=Free
  swing2LimitDegrees=45
  swing2Motion=Free
  twistLimitDegrees=45
  twistMotion=Free
 }
 {bone1Name=dyn_scarf, bone2Name=neck_01, swing1LimitDegrees=20, swing2LimitDegrees=10, twistLimitDegrees=1}
 {bone1Name=dyn_scarf_l, bone2Name=dyn_scarf, swing1LimitDegrees=20, swing2LimitDegrees=10, twistLimitDegrees=1}
 {bone1Name=dyn_scarf_r, bone2Name=dyn_scarf, swing1LimitDegrees=20, swing2LimitDegrees=10, twistLimitDegrees=1}
 {
  bone1Name=spine_05
  bone2Name=spine_04
  swing1LimitDegrees=45
  swing1Motion=Free
  swing2LimitDegrees=45
  swing2Motion=Free
  twistLimitDegrees=45
  twistMotion=Free
 }
]
