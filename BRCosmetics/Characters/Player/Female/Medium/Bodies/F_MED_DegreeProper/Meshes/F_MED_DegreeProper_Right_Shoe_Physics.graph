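# Physics asset for the right shoe mesh of the F_MED_DegreeProper female player body, setting up physics for its dangling laces.
# Two lace chains (left and right) each run from a small sphere base through two capsule segments.
# Bases have free rotation relative to calf_twist_01_r, while segment joints set swing limits and a very tight twist limit.
# Calf and foot capsules supply leg and shoe volumes, with the foot left unconstrained.
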PhysicsAsset /BRCosmetics/Characters/Player/Female/Medium/Bodies/F_MED_DegreeProper/Meshes/F_MED_DegreeProper_Right_Shoe_Physics
bodies {
 calf_twist_01_r [{shapeType=Capsule, length=8, radius=1.9916, rotation=<rot 84.9179 -0.0007 0.0013>}]
 dyn_lace_right_leg_l_01 [{
   shapeType=Capsule
   length=2.3599
   radius=0.6599
   shapeName=dyn_lace_right_leg_l_01_capsule
   center=<6.0499 0.0033 -0.1179>
   rotation=<rot 0 -83.4182 -64.4183>
  }]
 dyn_lace_right_leg_l_02 [{
   shapeType=Capsule
   length=1.9814
   radius=0.6535
   shapeName=dyn_lace_right_leg_l_02_capsule
   center=<5.4082 0.1845 -0.2557>
   rotation=<rot 0 87.84 -86.0369>
  }]
 dyn_lace_right_leg_l_base [{shapeType=Sphere, radius=0.505, shapeName=dyn_lace_right_leg_l_base_sphere}]
 dyn_lace_right_leg_r_01 [{
   shapeType=Capsule
   length=2.1802
   radius=0.7285
   shapeName=dyn_lace_right_leg_r_01_capsule
   center=<4.7554 0.3736 -0.2378>
   rotation=<rot 0 82.3496 -66.5365>
  }]
 dyn_lace_right_leg_r_02 [{
   shapeType=Capsule
   length=3.1643
   radius=0.9488
   shapeName=dyn_lace_right_leg_r_02_capsule
   center=<4.0461 -0.9294 0.6658>
   rotation=<rot 0 83.1596 -89.8295>
  }]
 dyn_lace_right_leg_r_base [{shapeType=Sphere, radius=0.505, shapeName=dyn_lace_right_leg_r_base_sphere}]
 foot_r [{shapeType=Capsule, length=15.9887, radius=6.0609, center=<7.1132 -9.0903 0.1505>, rotation=<rot 0 9.1929 -87.7817>}]
}
constraints [
 {bone1Name=dyn_lace_right_leg_r_02, bone2Name=dyn_lace_right_leg_r_01, swing1LimitDegrees=40, swing2LimitDegrees=20, twistLimitDegrees=1}
 {bone1Name=dyn_lace_right_leg_l_01, bone2Name=dyn_lace_right_leg_l_base, swing1LimitDegrees=40, swing2LimitDegrees=20, twistLimitDegrees=1}
 {bone1Name=dyn_lace_right_leg_l_02, bone2Name=dyn_lace_right_leg_l_01, swing1LimitDegrees=40, swing2LimitDegrees=20, twistLimitDegrees=1}
 {bone1Name=dyn_lace_right_leg_r_01, bone2Name=dyn_lace_right_leg_r_base, swing1LimitDegrees=40, swing2LimitDegrees=20, twistLimitDegrees=1}
 {
  bone1Name=dyn_lace_right_leg_l_base
  bone2Name=calf_twist_01_r
  swing1LimitDegrees=45
  swing1Motion=Free
  swing2LimitDegrees=45
  swing2Motion=Free
  twistLimitDegrees=45
  twistMotion=Free
 }
 {
  bone1Name=dyn_lace_right_leg_r_base
  bone2Name=calf_twist_01_r
  swing1LimitDegrees=45
  swing1Motion=Free
  swing2LimitDegrees=45
  swing2Motion=Free
  twistLimitDegrees=45
  twistMotion=Free
 }
]
