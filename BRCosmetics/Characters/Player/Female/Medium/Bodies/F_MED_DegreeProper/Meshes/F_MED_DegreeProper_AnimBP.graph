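# Post-process-style cosmetic AnimBP for the F_MED_DegreeProper female body, layering corrective and secondary motion onto the pose copied from its parent mesh.
# Corrects pelvis_corr (collapsed at LOD 2/3), runs a Control Rig, offsets lace bases and adjusts the right elbow for a front-end idle.
# RigidBody and chained AnimDynamics then simulate dynamic bones, including belts and laces.
# Shoe RigidBody and duplicate AnimDynamics chains are unconnected; computed calf values are unread here.
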AnimBlueprint /BRCosmetics/Characters/Player/Female/Medium/Bodies/F_MED_DegreeProper/Meshes/F_MED_DegreeProper_AnimBP
target_skeleton=&/F_MED_DegreeProper_Skeleton
graphs [
 AnimGraph {nodes [
   n1 = Anim Dynamics - Bone: dyn_belt_l(Alpha=1, ExternalForce="0,0,0") {
    angularDampingOverride=1
    bOverrideAngularDamping=true
    simulationSpace=RootRelative
    alphaBoolBlend {blendOption=Linear}
    boundBone {boneName=dyn_belt_l}
    physicsBodyDefinitions [{
      boundBone {boneName=dyn_belt_l}
      boxExtents=<10 10 10>
      constraintSetup {angularLimitsMax=<0 20 0>, angularLimitsMin=<0 -20 0>, angularTarget=<0 0 0>}
      localJointOffset=<1 0 0>
     }]
    relativeSpaceBone {boneName=lowerarm_l}
   }
   n2 = Anim Dynamics - Bone: dyn_belt_r(ComponentPose=@n1, Alpha=1, ExternalForce="0,0,0") {
    angularDampingOverride=1
    bOverrideAngularDamping=true
    simulationSpace=RootRelative
    alphaBoolBlend {blendOption=Linear}
    boundBone {boneName=dyn_belt_r}
    physicsBodyDefinitions [{
      boundBone {boneName=dyn_belt_r}
      boxExtents=<10 10 10>
      constraintSetup {angularLimitsMax=<0 20 0>, angularLimitsMin=<0 -20 0>, angularTarget=<0 0 0>}
      localJointOffset=<1 0 0>
     }]
    relativeSpaceBone {boneName=lowerarm_r}
   }
   n3 = Anim Dynamics - Bone: dyn_lace_left_leg_l_01(ComponentPose=@n2, Alpha=1, ExternalForce="0,0,0") {
    angularDampingOverride=1
    bOverrideAngularDamping=true
    simulationSpace=RootRelative
    alphaBoolBlend {blendOption=Linear}
    boundBone {boneName=dyn_lace_left_leg_l_01}
    physicsBodyDefinitions [{
      boundBone {boneName=dyn_lace_left_leg_l_01}
      boxExtents=<10 10 10>
      constraintSetup {angularLimitsMax=<0 0 20>, angularLimitsMin=<0 -20 -20>, angularTarget=<0 0 0>}
      localJointOffset=<1 0 0>
     }]
    relativeSpaceBone {boneName=lowerarm_l}
   }
   n4 = Anim Dynamics - Bone: dyn_lace_left_leg_l_02(ComponentPose=@n3, Alpha=1, ExternalForce="0,0,0") {
    angularDampingOverride=1
    bOverrideAngularDamping=true
    simulationSpace=RootRelative
    alphaBoolBlend {blendOption=Linear}
    boundBone {boneName=dyn_lace_left_leg_l_02}
    physicsBodyDefinitions [{
      boundBone {boneName=dyn_lace_left_leg_l_02}
      boxExtents=<10 10 10>
      constraintSetup {angularLimitsMax=<0 0 20>, angularLimitsMin=<0 -20 -20>, angularTarget=<0 0 0>}
      localJointOffset=<1 0 0>
     }]
    relativeSpaceBone {boneName=lowerarm_r}
   }
   n5 = Anim Dynamics - Bone: dyn_lace_left_leg_r_01(ComponentPose=@n4, Alpha=1, ExternalForce="0,0,0") {
    angularDampingOverride=1
    bOverrideAngularDamping=true
    simulationSpace=RootRelative
    alphaBoolBlend {blendOption=Linear}
    boundBone {boneName=dyn_lace_left_leg_r_01}
    physicsBodyDefinitions [{
      boundBone {boneName=dyn_lace_left_leg_r_01}
      boxExtents=<10 10 10>
      constraintSetup {angularLimitsMax=<0 20 20>, angularLimitsMin=<0 0 -20>, angularTarget=<0 0 0>}
      localJointOffset=<1 0 0>
     }]
    relativeSpaceBone {boneName=lowerarm_l}
   }
   n6 = Anim Dynamics - Bone: dyn_lace_left_leg_r_02(ComponentPose=@n5, Alpha=1, ExternalForce="0,0,0") {
    angularDampingOverride=1
    bOverrideAngularDamping=true
    simulationSpace=RootRelative
    alphaBoolBlend {blendOption=Linear}
    boundBone {boneName=dyn_lace_left_leg_r_02}
    physicsBodyDefinitions [{
      boundBone {boneName=dyn_lace_left_leg_r_02}
      boxExtents=<10 10 10>
      constraintSetup {angularLimitsMax=<0 20 20>, angularLimitsMin=<0 0 -20>, angularTarget=<0 0 0>}
      localJointOffset=<1 0 0>
     }]
    relativeSpaceBone {boneName=lowerarm_r}
   }
   n7 = Anim Dynamics - Bone: dyn_lace_right_leg_l_01(ComponentPose=@n6, Alpha=1, ExternalForce="0,0,0") {
    angularDampingOverride=1
    bOverrideAngularDamping=true
    simulationSpace=RootRelative
    alphaBoolBlend {blendOption=Linear}
    boundBone {boneName=dyn_lace_right_leg_l_01}
    physicsBodyDefinitions [{
      boundBone {boneName=dyn_lace_right_leg_l_01}
      boxExtents=<10 10 10>
      constraintSetup {angularLimitsMax=<0 0 20>, angularLimitsMin=<0 -20 -20>, angularTarget=<0 0 0>}
      localJointOffset=<1 0 0>
     }]
    relativeSpaceBone {boneName=lowerarm_l}
   }
   n8 = Anim Dynamics - Bone: dyn_lace_right_leg_l_02(ComponentPose=@n7, Alpha=1, ExternalForce="0,0,0") {
    angularDampingOverride=1
    bOverrideAngularDamping=true
    simulationSpace=RootRelative
    alphaBoolBlend {blendOption=Linear}
    boundBone {boneName=dyn_lace_right_leg_l_02}
    physicsBodyDefinitions [{
      boundBone {boneName=dyn_lace_right_leg_l_02}
      boxExtents=<10 10 10>
      constraintSetup {angularLimitsMax=<0 0 20>, angularLimitsMin=<0 -20 -20>, angularTarget=<0 0 0>}
      localJointOffset=<1 0 0>
     }]
    relativeSpaceBone {boneName=lowerarm_r}
   }
   n9 = Anim Dynamics - Bone: dyn_lace_right_leg_r_01(ComponentPose=@n8, Alpha=1, ExternalForce="0,0,0") {
    angularDampingOverride=1
    bOverrideAngularDamping=true
    simulationSpace=RootRelative
    alphaBoolBlend {blendOption=Linear}
    boundBone {boneName=dyn_lace_right_leg_r_01}
    physicsBodyDefinitions [{
      boundBone {boneName=dyn_lace_right_leg_r_01}
      boxExtents=<10 10 10>
      constraintSetup {angularLimitsMax=<0 20 20>, angularLimitsMin=<0 5 -20>, angularTarget=<0 0 0>}
      localJointOffset=<1 0 0>
     }]
    relativeSpaceBone {boneName=lowerarm_l}
   }
   Anim Dynamics - Bone: dyn_lace_right_leg_r_02(ComponentPose=@n9, Alpha=1, ExternalForce="0,0,0") {
    angularDampingOverride=1
    bOverrideAngularDamping=true
    simulationSpace=RootRelative
    alphaBoolBlend {blendOption=Linear}
    boundBone {boneName=dyn_lace_right_leg_r_02}
    physicsBodyDefinitions [{
      boundBone {boneName=dyn_lace_right_leg_r_02}
      boxExtents=<10 10 10>
      constraintSetup {angularLimitsMax=<0 20 20>, angularLimitsMin=<0 7 -20>, angularTarget=<0 0 0>}
      localJointOffset=<1 0 0>
     }]
    relativeSpaceBone {boneName=lowerarm_r}
   }
   n10 = RigidBody() {
    overridePhysicsAsset=&/F_MED_DegreeProper_Left_Shoe_Physics
    simulationSpace=BaseBoneSpace
    alphaBoolBlend {blendOption=Linear}
    baseBoneRef {boneName=calf_l}
    simSpaceSettings {maxAngularAcceleration=65, maxAngularVelocity=15, maxLinearAcceleration=1000, maxLinearVelocity=3, velocityScaleZ=0.1}
   }
   RigidBody(ComponentPose=@n10) {
    overridePhysicsAsset=&/F_MED_DegreeProper_Right_Shoe_Physics
    simulationSpace=BaseBoneSpace
    alphaBoolBlend {blendOption=Linear}
    baseBoneRef {boneName=calf_r}
    simSpaceSettings {maxAngularAcceleration=65, maxAngularVelocity=15, maxLinearAcceleration=1000, maxLinearVelocity=3, velocityScaleZ=0.1}
   }
   n11 = "Constraint - pelvis_corr (P)"(ComponentPose=Local To Component(LocalPose=Copy Pose From Mesh() {bUseAttachedParent=true}), Alpha=1, ConstraintWeights_0=1) {
    alphaBoolBlend {blendOption=Linear}
    boneToModify {boneName=pelvis_corr}
    constraintSetup [{targetBone {boneName=pelvis}}]
    constraintWeights [1]
   }
   n12 = "Transform (Modify) Bone - Bone: pelvis_corr"!(
    ComponentPose=@n11
    bAlphaBoolEnabled:bool="Get LOD 2,3"() -> ("LOD 2,3":bool)
    Translation="0.000,5.030,-11.403"
    Rotation="0,0,0"
    Scale="0,0,0"
   ) -> (Pose) {
    alphaInputType=Bool
    rotationMode=BMM_Additive
    scaleMode=BMM_Replace
    scaleSpace=BCS_BoneSpace
    translationMode=BMM_Additive
    alphaBoolBlend {blendOption=Linear}
    boneToModify {boneName=pelvis_corr}
    scale=<0 0 0>
    translation {y=5.03, z=-11.403}
   }
   n13 = Control Rig(Source=Component To Local(ComponentPose=@n12)) {
    alphaBoolBlend {blendOption=Linear}
    controlRigAssetReference {blueprintRigClass=&/F_MED_DegreeProper_CtrlRig.F_MED_DegreeProper_CtrlRig_C}
   }
   n14 = "Transform (Modify) Bone - Bone: deform_elbow_in_r"!(
    ComponentPose=Local To Component(LocalPose=@n13)
    bAlphaBoolEnabled:bool=Get Custom idle() -> (Custom idle:bool)
    Translation="7.215,-0.000,-2.492"
    Rotation="0.000,28.456,-0.000"
    Scale="1,1,1"
   ) -> (Pose) {
    alphaInputType=Bool
    rotationMode=BMM_Additive
    translationMode=BMM_Additive
    alphaBoolBlend {blendOption=Linear}
    boneToModify {boneName=deform_elbow_in_r}
    rotation {yaw=28.456}
    translation {x=7.215, z=-2.492}
   }
   n15 = "Transform (Modify) Bone - Bone: dyn_lace_left_leg_l_base"(ComponentPose=@n14, Alpha=1, Translation="-0.000,-0.000,-2.492", Rotation="0,0,0", Scale="1,1,1") {translationMode=BMM_Additive, alphaBoolBlend {blendOption=Linear}, boneToModify {boneName=dyn_lace_left_leg_l_base}, translation {z=-2.492}}
   n16 = "Transform (Modify) Bone - Bone: dyn_lace_left_leg_r_base"(ComponentPose=@n15, Alpha=1, Translation="-0.000,-0.000,-5.046", Rotation="0,0,0", Scale="1,1,1") {translationMode=BMM_Additive, alphaBoolBlend {blendOption=Linear}, boneToModify {boneName=dyn_lace_left_leg_r_base}, translation {z=-5.046}}
   n17 = "Transform (Modify) Bone - Bone: dyn_lace_right_leg_l_base"(ComponentPose=@n16, Alpha=1, Translation="-4.631,1.364,-1.524", Rotation="0,0,0", Scale="1,1,1") {
    translationMode=BMM_Additive
    alphaBoolBlend {blendOption=Linear}
    boneToModify {boneName=dyn_lace_right_leg_l_base}
    translation=<-4.631 1.364 -1.524>
   }
   n18 = "Transform (Modify) Bone - Bone: dyn_lace_right_leg_r_base"(ComponentPose=@n17, Alpha=1, Translation="-3,-0.589,-0.946", Rotation="0,0,0", Scale="1,1,1") {
    translationMode=BMM_Additive
    alphaBoolBlend {blendOption=Linear}
    boneToModify {boneName=dyn_lace_right_leg_r_base}
    translation=<-3 -0.589 -0.946>
   }
   n19 = RigidBody(ComponentPose=@n18) {
    overridePhysicsAsset=&/F_MED_DegreeProper_Physics
    simulationSpace=BaseBoneSpace
    alphaBoolBlend {blendOption=Linear}
    baseBoneRef {boneName=spine_05}
    simSpaceSettings {maxAngularAcceleration=65, maxAngularVelocity=15, maxLinearAcceleration=1000, maxLinearVelocity=3, velocityScaleZ=0.1}
   }
   n20 = Anim Dynamics - Bone: dyn_belt_l(ComponentPose=@n19, Alpha=1, ExternalForce="0,0,0") {
    angularDampingOverride=1
    bOverrideAngularDamping=true
    simulationSpace=RootRelative
    alphaBoolBlend {blendOption=Linear}
    boundBone {boneName=dyn_belt_l}
    physicsBodyDefinitions [{
      boundBone {boneName=dyn_belt_l}
      boxExtents=<10 10 10>
      constraintSetup {angularLimitsMax=<0 20 0>, angularLimitsMin=<0 -20 0>, angularTarget=<0 0 0>}
      localJointOffset=<1 0 0>
     }]
    relativeSpaceBone {boneName=lowerarm_l}
    simSpaceSettings {simSpaceAngularAlpha=0.3}
   }
   n21 = Anim Dynamics - Bone: dyn_belt_r(ComponentPose=@n20, Alpha=1, ExternalForce="0,0,0") {
    angularDampingOverride=1
    bOverrideAngularDamping=true
    simulationSpace=RootRelative
    alphaBoolBlend {blendOption=Linear}
    boundBone {boneName=dyn_belt_r}
    physicsBodyDefinitions [{
      boundBone {boneName=dyn_belt_r}
      boxExtents=<10 10 10>
      constraintSetup {angularLimitsMax=<0 20 0>, angularLimitsMin=<0 -20 0>, angularTarget=<0 0 0>}
      localJointOffset=<1 0 0>
     }]
    relativeSpaceBone {boneName=lowerarm_r}
    simSpaceSettings {simSpaceAngularAlpha=0.3}
   }
   n22 = Anim Dynamics - Bone: dyn_lace_left_leg_l_01(ComponentPose=@n21, Alpha=1, ExternalForce="0,0,0") {
    angularDampingOverride=1
    bOverrideAngularDamping=true
    simulationSpace=RootRelative
    alphaBoolBlend {blendOption=Linear}
    boundBone {boneName=dyn_lace_left_leg_l_01}
    physicsBodyDefinitions [{
      boundBone {boneName=dyn_lace_left_leg_l_01}
      boxExtents=<10 10 10>
      constraintSetup {angularLimitsMax=<0 0 20>, angularLimitsMin=<0 -20 -20>, angularTarget=<0 0 0>}
      localJointOffset=<1 0 0>
     }]
    relativeSpaceBone {boneName=lowerarm_l}
    simSpaceSettings {simSpaceAngularAlpha=0.3}
   }
   n23 = Anim Dynamics - Bone: dyn_lace_left_leg_l_02(ComponentPose=@n22, Alpha=1, ExternalForce="0,0,0") {
    angularDampingOverride=1
    bOverrideAngularDamping=true
    simulationSpace=RootRelative
    alphaBoolBlend {blendOption=Linear}
    boundBone {boneName=dyn_lace_left_leg_l_02}
    physicsBodyDefinitions [{
      boundBone {boneName=dyn_lace_left_leg_l_02}
      boxExtents=<10 10 10>
      constraintSetup {angularLimitsMax=<0 0 20>, angularLimitsMin=<0 -20 -20>, angularTarget=<0 0 0>}
      localJointOffset=<1 0 0>
     }]
    relativeSpaceBone {boneName=lowerarm_r}
    simSpaceSettings {simSpaceAngularAlpha=0.3}
   }
   n24 = Anim Dynamics - Bone: dyn_lace_left_leg_r_01(ComponentPose=@n23, Alpha=1, ExternalForce="0,0,0") {
    angularDampingOverride=1
    bOverrideAngularDamping=true
    simulationSpace=RootRelative
    alphaBoolBlend {blendOption=Linear}
    boundBone {boneName=dyn_lace_left_leg_r_01}
    physicsBodyDefinitions [{
      boundBone {boneName=dyn_lace_left_leg_r_01}
      boxExtents=<10 10 10>
      constraintSetup {angularLimitsMax=<0 20 20>, angularLimitsMin=<0 0 -20>, angularTarget=<0 0 0>}
      localJointOffset=<1 0 0>
     }]
    relativeSpaceBone {boneName=lowerarm_l}
    simSpaceSettings {simSpaceAngularAlpha=0.3}
   }
   n25 = Anim Dynamics - Bone: dyn_lace_left_leg_r_02(ComponentPose=@n24, Alpha=1, ExternalForce="0,0,0") {
    angularDampingOverride=1
    bOverrideAngularDamping=true
    simulationSpace=RootRelative
    alphaBoolBlend {blendOption=Linear}
    boundBone {boneName=dyn_lace_left_leg_r_02}
    physicsBodyDefinitions [{
      boundBone {boneName=dyn_lace_left_leg_r_02}
      boxExtents=<10 10 10>
      constraintSetup {angularLimitsMax=<0 20 20>, angularLimitsMin=<0 0 -20>, angularTarget=<0 0 0>}
      localJointOffset=<1 0 0>
     }]
    relativeSpaceBone {boneName=lowerarm_r}
    simSpaceSettings {simSpaceAngularAlpha=0.3}
   }
   n26 = Anim Dynamics - Bone: dyn_lace_right_leg_l_01(ComponentPose=@n25, Alpha=1, ExternalForce="0,0,0") {
    angularDampingOverride=1
    bOverrideAngularDamping=true
    simulationSpace=RootRelative
    alphaBoolBlend {blendOption=Linear}
    boundBone {boneName=dyn_lace_right_leg_l_01}
    physicsBodyDefinitions [{
      boundBone {boneName=dyn_lace_right_leg_l_01}
      boxExtents=<10 10 10>
      constraintSetup {angularLimitsMax=<0 0 20>, angularLimitsMin=<0 -20 -20>, angularTarget=<0 0 0>}
      localJointOffset=<1 0 0>
     }]
    relativeSpaceBone {boneName=lowerarm_l}
    simSpaceSettings {simSpaceAngularAlpha=0.3}
   }
   n27 = Anim Dynamics - Bone: dyn_lace_right_leg_l_02(ComponentPose=@n26, Alpha=1, ExternalForce="0,0,0") {
    angularDampingOverride=1
    bOverrideAngularDamping=true
    simulationSpace=RootRelative
    alphaBoolBlend {blendOption=Linear}
    boundBone {boneName=dyn_lace_right_leg_l_02}
    physicsBodyDefinitions [{
      boundBone {boneName=dyn_lace_right_leg_l_02}
      boxExtents=<10 10 10>
      constraintSetup {angularLimitsMax=<0 0 20>, angularLimitsMin=<0 -20 -20>, angularTarget=<0 0 0>}
      localJointOffset=<1 0 0>
     }]
    relativeSpaceBone {boneName=lowerarm_r}
    simSpaceSettings {simSpaceAngularAlpha=0.3}
   }
   n28 = Anim Dynamics - Bone: dyn_lace_right_leg_r_01(ComponentPose=@n27, Alpha=1, ExternalForce="0,0,0") {
    angularDampingOverride=1
    bOverrideAngularDamping=true
    simulationSpace=RootRelative
    alphaBoolBlend {blendOption=Linear}
    boundBone {boneName=dyn_lace_right_leg_r_01}
    physicsBodyDefinitions [{
      boundBone {boneName=dyn_lace_right_leg_r_01}
      boxExtents=<10 10 10>
      constraintSetup {angularLimitsMax=<0 20 20>, angularLimitsMin=<0 5 -20>, angularTarget=<0 0 0>}
      localJointOffset=<1 0 0>
     }]
    relativeSpaceBone {boneName=lowerarm_l}
    simSpaceSettings {simSpaceAngularAlpha=0.3}
   }
   n29 = Anim Dynamics - Bone: dyn_lace_right_leg_r_02(ComponentPose=@n28, Alpha=1, ExternalForce="0,0,0") {
    angularDampingOverride=1
    bOverrideAngularDamping=true
    simulationSpace=RootRelative
    alphaBoolBlend {blendOption=Linear}
    boundBone {boneName=dyn_lace_right_leg_r_02}
    physicsBodyDefinitions [{
      boundBone {boneName=dyn_lace_right_leg_r_02}
      boxExtents=<10 10 10>
      constraintSetup {angularLimitsMax=<0 20 20>, angularLimitsMin=<0 7 -20>, angularTarget=<0 0 0>}
      localJointOffset=<1 0 0>
     }]
    relativeSpaceBone {boneName=lowerarm_r}
    simSpaceSettings {simSpaceAngularAlpha=0.3}
   }
   Output Pose(Result=Component To Local(ComponentPose=@n29)) {layerGroup=None}
  ]}
 EventGraph {nodes [
   Event BlueprintUpdateAnimation()
   n1 = GetOwnerSkeletalMeshComponent()
   > Set Calf_rot_L(Calf_rot_L:double=MapRangeClamped(
     Value=GetDeltaTransformFromRefPose!(self=@n1, BoneName=foot_l, BaseName=calf_l) -> (
      ReturnValue_Location:Vector
      ReturnValue_Rotation_Roll:float
      ReturnValue_Rotation_Pitch:float
      ReturnValue_Rotation_Yaw:float
      ReturnValue_Scale:Vector
     ).ReturnValue_Rotation_Yaw
     InRangeA=0
     InRangeB=60
    )) -> (Output_Get:double)
   n2 = GetOwnerSkeletalMeshComponent()
   > Set Calf_rot_R(Calf_rot_R:double=MapRangeClamped(Value=GetDeltaTransformFromRefPose(self=@n2, BoneName=foot_r, BaseName=calf_r).ReturnValue_Rotation_Yaw, InRangeA=0, InRangeB=60)) -> (Output_Get:double)
   n3 = GetDeltaTransformFromRefPose(self=GetOwnerSkeletalMeshComponent(), BoneName=lowerarm_l, BaseName=upperarm_l)
   > Set Custom idle(Custom idle:bool=AND Boolean(A=Get bIsInFrontEnd(), B="float > float"(A=@n3.ReturnValue_Rotation_Yaw, B=80))) -> (Output_Get:bool)
   n4 = Get CurrentLOD()
   > "Set LOD 2,3"("LOD 2,3":bool=OR Boolean(A="Equal (Integer)"(A=@n4, B=2), B="Equal (Integer)"(A=@n4, B=3))) -> (Output_Get:bool)
  ]}
]
variables [
 Calf_rot_L {type=number}
 Calf_rot_R {type=number}
 Custom idle {type=boolean}
 "LOD 2,3" {type=boolean}
]
