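# Physics asset for the face accessory mesh of the female medium DegreeProper cosmetic body, covering its dynamic hair and front bones.
# Sphere bodies on the front, mask-hair, side and top-hair bones each constrain directly to the head with tight swing and near-locked twist limits.
# Capsules on the head, neck, upper spine, clavicles and upper arms form the body volumes, linked by free-motion constraints.
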PhysicsAsset /BRCosmetics/Characters/Player/Female/Medium/Bodies/F_MED_DegreeProper/Meshes/Parts/F_MED_DegreeProper_FaceAcc_Physics
bodies {
 clavicle_l [{shapeType=Capsule, length=9, radius=7, shapeName=clavicle_l_capsule, center=<3.4946 0 -2.1822>, rotation=<rot 84.8961 -0.0014 -0.0014>}]
 clavicle_r [{shapeType=Capsule, length=9, radius=7, shapeName=clavicle_r_capsule, center=<-3.379 0 2.1223>, rotation=<rot 84.8961 -0.0014 -0.0014>}]
 dyn_front_01 [{shapeType=Sphere, radius=1, shapeName=dyn_front_01_sphere, center=<5 0 0>}]
 dyn_front_02 [{shapeType=Sphere, radius=1, shapeName=dyn_front_02_sphere, center=<5 0 0>}]
 dyn_front_03 [{shapeType=Sphere, radius=1, shapeName=dyn_front_03_sphere, center=<5 0 0>}]
 dyn_front_04 [{shapeType=Sphere, radius=1, shapeName=dyn_front_04_sphere, center=<5 0 0>}]
 dyn_front_05 [{shapeType=Sphere, radius=1, shapeName=dyn_front_05_sphere, center=<5 0 0>}]
 dyn_front_06 [{shapeType=Sphere, radius=1, shapeName=dyn_front_06_sphere, center=<5 0 0>}]
 dyn_front_mask_hair [{shapeType=Sphere, radius=3.0435, shapeName=dyn_front_mask_hair_sphere, center=<3.8616 0.1155 -0.0296>}]
 dyn_side_hair_l [{shapeType=Sphere, radius=2, shapeName=dyn_side_hair_l_sphere, center=<9.4809 0.2467 -0.8711>}]
 dyn_side_hair_r [{shapeType=Sphere, radius=2, shapeName=dyn_side_hair_r_sphere, center=<12.007 0.6795 1.6666>}]
 dyn_top_hair_01 [{shapeType=Sphere, radius=2, shapeName=dyn_top_hair_01_sphere, center=<4.6618 0.3871 1.2655>}]
 dyn_top_hair_02 [{shapeType=Sphere, radius=2, shapeName=dyn_top_hair_02_sphere, center=<3.1959 0.0858 0.193>}]
 dyn_top_hair_03 [{shapeType=Sphere, radius=2, shapeName=dyn_top_hair_03_sphere, center=<4.5304 1.693 1.111>}]
 dyn_top_hair_04 [{shapeType=Sphere, radius=2, shapeName=dyn_top_hair_04_sphere, center=<3.1096 0.3679 -1.7696>}]
 head [{shapeType=Capsule, length=0, radius=1, center=<7.0249 2.9672 0.1663>, rotation=<rot 4.5558 67.4239 -90>}]
 neck_01 [{shapeType=Capsule, length=5, radius=5.5, shapeName=neck_01_capsule, center=<1.1101 0 0>, rotation=<rot 90 0 0>}]
 neck_02 [{shapeType=Capsule, length=5, radius=5.5, shapeName=neck_02_capsule, center=<1.1101 0 0>, rotation=<rot 90 0 0>}]
 spine_05 [{shapeType=Capsule, length=16.505, radius=7.505, shapeName=spine_05_capsule, center=<2.9277 -0.136 0>, rotation=<rot 0 -90 0>}]
 upperarm_l [{shapeType=Capsule, length=24.5, radius=5, shapeName=upperarm_l_capsule, center=<14.4434 0 0>, rotation=<rot 90 0 0>}]
 upperarm_r [{shapeType=Capsule, length=24.5, radius=5, shapeName=upperarm_r_capsule, center=<-14.9868 0 0>, rotation=<rot 90 0 0>}]
}
constraints [
 {bone1Name=dyn_top_hair_04, bone2Name=head, swing1LimitDegrees=10, swing2LimitDegrees=10, twistLimitDegrees=1}
 {bone1Name=dyn_top_hair_03, bone2Name=head, swing1LimitDegrees=10, swing2LimitDegrees=10, twistLimitDegrees=1}
 {bone1Name=dyn_top_hair_02, bone2Name=head, swing1LimitDegrees=10, swing2LimitDegrees=10, twistLimitDegrees=1}
 {bone1Name=dyn_top_hair_01, bone2Name=head, swing1LimitDegrees=10, swing2LimitDegrees=10, twistLimitDegrees=1}
 {bone1Name=dyn_side_hair_r, bone2Name=head, swing1LimitDegrees=25, swing2LimitDegrees=10, twistLimitDegrees=1}
 {bone1Name=dyn_side_hair_l, bone2Name=head, swing1LimitDegrees=30, swing2LimitDegrees=10, twistLimitDegrees=1}
 {bone1Name=dyn_front_mask_hair, bone2Name=head, swing1LimitDegrees=10, swing2LimitDegrees=10, twistLimitDegrees=1}
 {
  bone1Name=head
  bone2Name=neck_02
  swing1LimitDegrees=45
  swing1Motion=Free
  swing2LimitDegrees=45
  swing2Motion=Free
  twistLimitDegrees=45
  twistMotion=Free
 }
 {
  bone1Name=clavicle_l
  bone2Name=spine_05
  swing1LimitDegrees=45
  swing1Motion=Free
  swing2LimitDegrees=45
  swing2Motion=Free
  twistLimitDegrees=45
  twistMotion=Free
 }
 {
  bone1Name=clavicle_r
  bone2Name=spine_05
  swing1LimitDegrees=45
  swing1Motion=Free
  swing2LimitDegrees=45
  swing2Motion=Free
  twistLimitDegrees=45
  twistMotion=Free
 }
 {bone1Name=dyn_front_01, bone2Name=head, swing1LimitDegrees=3, swing2LimitDegrees=20, twistLimitDegrees=1}
 {bone1Name=dyn_front_02, bone2Name=head, swing1LimitDegrees=3, swing2LimitDegrees=20, twistLimitDegrees=1}
 {bone1Name=dyn_front_03, bone2Name=head, swing1LimitDegrees=3, swing2LimitDegrees=20, twistLimitDegrees=1}
 {bone1Name=dyn_front_04, bone2Name=head, swing1LimitDegrees=3, swing2LimitDegrees=20, twistLimitDegrees=1}
 {bone1Name=dyn_front_05, bone2Name=head, swing1LimitDegrees=3, swing2LimitDegrees=20, twistLimitDegrees=1}
 {bone1Name=dyn_front_06, bone2Name=head, swing1LimitDegrees=3, swing2LimitDegrees=20, twistLimitDegrees=1}
 {
  bone1Name=upperarm_r
  bone2Name=clavicle_r
  swing1LimitDegrees=45
  swing1Motion=Free
  swing2LimitDegrees=45
  swing2Motion=Free
  twistLimitDegrees=45
  twistMotion=Free
 }
 {
  bone1Name=upperarm_l
  bone2Name=clavicle_l
  swing1LimitDegrees=45
  swing1Motion=Free
  swing2LimitDegrees=45
  swing2Motion=Free
  twistLimitDegrees=45
  twistMotion=Free
 }
 {
  bone1Name=neck_02
  bone2Name=neck_01
  swing1LimitDegrees=45
  swing1Motion=Free
  swing2LimitDegrees=45
  swing2Motion=Free
  twistLimitDegrees=45
  twistMotion=Free
 }
 {
  bone1Name=neck_01
  bone2Name=spine_05
  swing1LimitDegrees=45
  swing1Motion=Free
  swing2LimitDegrees=45
  swing2Motion=Free
  twistLimitDegrees=45
  twistMotion=Free
 }
]
