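# Control Rig for the DegreeProper female medium character's back hair, posing dyn_hair master bones procedurally from body motion.
# Three spherical pose readers track the head against spine_05: one rotates and offsets center and side hair masters, the others handle right and left, adding X-scale squash.
# Two clavicle readers add side-hair rotation, multiplied by the matching head reader so it applies only when both poses are active.
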ControlRigBlueprint /BRCosmetics/Characters/Player/Female/Medium/Bodies/F_MED_DegreeProper/Meshes/Parts/F_MED_DegreeProper_backHair_FaceAcc_CtrlRig
graphs [
 Multiply_5_1_ContainedGraph {nodes [
   n1 = RigVMFunctionEntryNode()
   RigVMFunctionReturnNode(Result=&RigVMFunction_MathFloatMul(B=@n1.C, A=&RigVMFunction_MathFloatMul(A=@n1.A, B=@n1.B) {template="Multiply::Execute(in A,in B,out Result)"}.Result) {template="Multiply::Execute(in A,in B,out Result)"}.Result)
  ]}
 Multiply_5_1_ContainedGraph {nodes [
   n1 = RigVMFunctionEntryNode()
   RigVMFunctionReturnNode(Result=&RigVMFunction_MathFloatMul(B=@n1.C, A=&RigVMFunction_MathFloatMul(A=@n1.A, B=@n1.B) {template="Multiply::Execute(in A,in B,out Result)"}.Result) {template="Multiply::Execute(in A,in B,out Result)"}.Result)
  ]}
 RigVMFunctionLibrary {links [], nodes []}
 RigVMModel {nodes [
   RigUnit_BeginExecution = &RigUnit_BeginExecution()
   SequenceAggregate = RigVMAggregateNode(ExecuteContext=@RigUnit_BeginExecution.ExecutePin)
   n1 = &RigUnit_SphericalPoseReader(
    ActiveRegionScaleFactors=(PositiveWidth=0.1,NegativeWidth=0.1,PositiveHeight=0.1,NegativeHeight=0.1)
    ActiveRegionSize=0.1
    Debug=(bDrawDebug=false,DebugScale=25,DebugSegments=20,DebugThickness=0.25)
    DriverAxis=(X=0,Y=-1,Z=0)
    DriverItem=(Type=Bone,Name="head")
    FalloffRegionScaleFactors=(PositiveWidth=1,NegativeWidth=1,PositiveHeight=1,NegativeHeight=1)
    FalloffSize=0.5
    FlipWidthScaling=true
    OptionalParentItem=(Type=Bone,Name="spine_05")
    RotationOffset=(X=90,Y=180,Z=90)
    ExecutePin=@SequenceAggregate.A
   )
   Multiply = &RigVMFunction_MathDoubleMul(A=-120, B=@n1.OutputParam)
   n2 = &RigVMFunction_MathDoubleMul(A=-60, B=@n1.OutputParam)
   n3 = &RigVMFunction_MathDoubleMul(A=0, B=@n1.OutputParam)
   n4 = &RigVMFunction_MathDoubleMul(A=-120, B=@n1.OutputParam)
   n5 = &RigVMFunction_MathDoubleMul(A=0, B=@n1.OutputParam)
   n6 = &RigVMFunction_MathDoubleSub(A=1, B=&RigVMFunction_MathFloatMul(A=0, B=@n1.OutputParam) {template="Multiply::Execute(in A,in B,out Result)"}.Result)
   n7 = &RigVMFunction_MathDoubleSub(A=1, B=&RigVMFunction_MathFloatMul(A=0, B=@n1.OutputParam) {template="Multiply::Execute(in A,in B,out Result)"}.Result)
   > &RigUnit_ModifyTransforms(
    ItemToModify="((Item=(Type=Bone,Name=\"dyn_hair_c_Master\"),Transform=(Rotation=(X=0,Y=0,Z=0,W=1),Translation=(X=0,Y=0,Z=0),Scale3D=(X=1,Y=1,Z=1))),(Item=(Type=Bone,Name=\"dyn_hair_l_B_Master\"),Transform=(Rotation=(X=0,Y=0,Z=0,W=1),Translation=(X=0,Y=0,Z=0),Scale3D=(X=1,Y= ...(truncated)"
    Weight=1
    ItemToModify.0.Transform.Rotation=&RigVMFunction_MathQuaternionFromEuler(Euler=(X=0,Y=0,Z=0), Euler.Y=@Multiply.Result).Result
    ItemToModify.3.Transform.Rotation=&RigVMFunction_MathQuaternionFromEuler(Euler=(X=0,Y=0,Z=0), Euler.Y=@n2.Result, Euler.Z=@n3.Result).Result
    ItemToModify.2.Transform.Rotation=&RigVMFunction_MathQuaternionFromEuler(Euler=(X=0,Y=0,Z=0), Euler.Y=@n4.Result, Euler.Z=@n5.Result).Result
    ItemToModify.1.Transform.Rotation=&RigVMFunction_MathQuaternionFromEuler(Euler=(X=0,Y=0,Z=0), Euler.Y=@n4.Result, Euler.Z=&RigVMFunction_MathDoubleMul(B=-1, A=@n5.Result).Result).Result
    ItemToModify.4.Transform.Rotation=&RigVMFunction_MathQuaternionFromEuler(Euler=(X=0,Y=0,Z=0), Euler.Y=@n2.Result, Euler.Z=&RigVMFunction_MathDoubleMul(B=-1, A=@n3.Result).Result).Result
    ItemToModify.3.Transform.Scale3D.X=@n6.Result
    ItemToModify.4.Transform.Scale3D.X=@n6.Result
    ItemToModify.1.Transform.Scale3D.X=@n7.Result
    ItemToModify.2.Transform.Scale3D.X=@n7.Result
   )
   > &RigUnit_ModifyTransforms(
    ItemToModify="((Item=(Type=Bone,Name=\"dyn_hair_c_Master\"),Transform=(Rotation=(X=0,Y=0,Z=0,W=1),Translation=(X=0,Y=0,Z=5),Scale3D=(X=1,Y=1,Z=1))),(Item=(Type=Bone,Name=\"dyn_hair_l_B_Master\"),Transform=(Rotation=(X=0,Y=0,Z=0,W=1),Translation=(X=0,Y=0,Z=5),Scale3D=(X=1,Y= ...(truncated)"
    Weight=@n1.OutputParam
   )
   n8 = &RigUnit_SphericalPoseReader(
    ActiveRegionScaleFactors=(PositiveWidth=0.1,NegativeWidth=0.1,PositiveHeight=0.1,NegativeHeight=0.1)
    ActiveRegionSize=0.1
    Debug=(bDrawDebug=false,DebugScale=25,DebugSegments=20,DebugThickness=0.25)
    DriverAxis=(X=0,Y=0,Z=1)
    DriverItem=(Type=Bone,Name="head")
    FalloffRegionScaleFactors=(PositiveWidth=1,NegativeWidth=1,PositiveHeight=1,NegativeHeight=1)
    FalloffSize=0.5
    FlipWidthScaling=true
    OptionalParentItem=(Type=Bone,Name="spine_05")
    RotationOffset=(X=0,Y=90,Z=0)
    ExecutePin=@SequenceAggregate.B
   )
   n9 = &RigVMFunction_MathQuaternionFromEuler(Euler=(X=0,Y=0,Z=0), Euler.Y=&RigVMFunction_MathDoubleMul(A=45, B=@n8.OutputParam).Result)
   Subtract = &RigVMFunction_MathDoubleSub(A=1, B=&RigVMFunction_MathFloatMul(A=0.5, B=@n8.OutputParam) {template="Multiply::Execute(in A,in B,out Result)"}.Result)
   > &RigUnit_ModifyTransforms(
    ItemToModify="((Item=(Type=Bone,Name=\"dyn_hair_r_A_Master\"),Transform=(Rotation=(X=0,Y=0,Z=0,W=1),Translation=(X=0,Y=0,Z=0),Scale3D=(X=1,Y=1,Z=1))),(Item=(Type=Bone,Name=\"dyn_hair_r_B_Master\"),Transform=(Rotation=(X=0,Y=0,Z=0,W=1),Translation=(X=0,Y=0,Z=0),Scale3D=(X=1, ...(truncated)"
    Weight=1
    ItemToModify.0.Transform.Rotation=&RigVMFunction_MathQuaternionFromEuler(Euler=(X=0,Y=0,Z=0), Euler.Y=&RigVMFunction_MathDoubleMul(A=-150, B=@n8.OutputParam).Result).Result
    ItemToModify.2.Transform.Rotation=&RigVMFunction_MathQuaternionFromEuler(
     Euler=(X=0,Y=0,Z=0)
     Euler.Y=&RigVMFunction_MathDoubleMul(A=-120, B=@n8.OutputParam).Result
     Euler.X=&RigVMFunction_MathDoubleMul(A=-90, B=@n8.OutputParam).Result
    ).Result
    ItemToModify.1.Transform.Rotation=&RigVMFunction_MathQuaternionFromEuler(
     Euler=(X=0,Y=0,Z=0)
     Euler.Y=&RigVMFunction_MathDoubleMul(A=-150, B=@n8.OutputParam).Result
     Euler.Z=&RigVMFunction_MathDoubleMul(A=60, B=@n8.OutputParam).Result
    ).Result
    ItemToModify.3.Transform.Rotation=@n9.Result
    ItemToModify.4.Transform.Rotation=@n9.Result
    ItemToModify.0.Transform.Scale3D.X=@Subtract.Result
   )
   n10 = &RigUnit_SphericalPoseReader(
    ActiveRegionScaleFactors=(PositiveWidth=0.1,NegativeWidth=0.1,PositiveHeight=0.1,NegativeHeight=0.1)
    ActiveRegionSize=0.1
    Debug=(bDrawDebug=false,DebugScale=25,DebugSegments=20,DebugThickness=0.25)
    DriverAxis=(X=0,Y=0,Z=-1)
    DriverItem=(Type=Bone,Name="head")
    FalloffRegionScaleFactors=(PositiveWidth=1,NegativeWidth=1,PositiveHeight=1,NegativeHeight=1)
    FalloffSize=0.5
    FlipWidthScaling=true
    OptionalParentItem=(Type=Bone,Name="spine_05")
    RotationOffset=(X=180,Y=-90,Z=0)
    ExecutePin=@SequenceAggregate.C
   )
   > &RigUnit_ModifyTransforms(
    ItemToModify="((Item=(Type=Bone,Name=\"dyn_hair_l_A_Master\"),Transform=(Rotation=(X=0,Y=0,Z=0,W=1),Translation=(X=0,Y=0,Z=0),Scale3D=(X=1,Y=1,Z=1))),(Item=(Type=Bone,Name=\"dyn_hair_l_B_Master\"),Transform=(Rotation=(X=0,Y=0,Z=0,W=1),Translation=(X=0,Y=0,Z=0),Scale3D=(X=1, ...(truncated)"
    Weight=1
    ItemToModify.0.Transform.Rotation=&RigVMFunction_MathQuaternionFromEuler(Euler=(X=0,Y=0,Z=0), Euler.Y=&RigVMFunction_MathDoubleMul(A=-150, B=@n10.OutputParam).Result).Result
    ItemToModify.2.Transform.Rotation=&RigVMFunction_MathQuaternionFromEuler(
     Euler=(X=0,Y=0,Z=0)
     Euler.Y=&RigVMFunction_MathDoubleMul(A=-120, B=@n10.OutputParam).Result
     Euler.X=&RigVMFunction_MathDoubleMul(A=90, B=@n10.OutputParam).Result
    ).Result
    ItemToModify.1.Transform.Rotation=&RigVMFunction_MathQuaternionFromEuler(
     Euler=(X=0,Y=0,Z=0)
     Euler.Y=&RigVMFunction_MathDoubleMul(A=-150, B=@n10.OutputParam).Result
     Euler.Z=&RigVMFunction_MathDoubleMul(A=-60, B=@n10.OutputParam).Result
    ).Result
    ItemToModify.3.Transform.Rotation=&RigVMFunction_MathQuaternionFromEuler(Euler=(X=0,Y=0,Z=0), Euler.Y=&RigVMFunction_MathDoubleMul(A=45, B=@n10.OutputParam).Result).Result
    ItemToModify.0.Transform.Scale3D.X=&RigVMFunction_MathDoubleSub(A=1, B=&RigVMFunction_MathFloatMul(A=0.75, B=@n10.OutputParam) {template="Multiply::Execute(in A,in B,out Result)"}.Result).Result
   )
   n11 = &RigUnit_SphericalPoseReader(
    ActiveRegionScaleFactors=(PositiveWidth=2.5,NegativeWidth=3,PositiveHeight=3,NegativeHeight=3)
    ActiveRegionSize=0.1
    Debug=(bDrawDebug=false,DebugScale=25,DebugSegments=20,DebugThickness=0.25)
    DriverAxis=(X=-1,Y=0,Z=0)
    DriverItem=(Type=Bone,Name="clavicle_r")
    FalloffRegionScaleFactors=(PositiveWidth=1,NegativeWidth=1,PositiveHeight=1,NegativeHeight=1)
    FalloffSize=0.5
    FlipHeightScaling=true
    OptionalParentItem=(Type=Bone)
    RotationOffset=(X=180,Y=0,Z=180)
    ExecutePin=@SequenceAggregate.D
   )
   Multiply_5 = RigVMAggregateNode(A=-60, C=@n8.OutputParam, B=@n11.OutputParam)
   RigVMFunction_MathQuaternionFromEuler = &RigVMFunction_MathQuaternionFromEuler(Euler=(X=0,Y=0,Z=0), Euler.Y=@Multiply_5.Result)
   > &RigUnit_ModifyTransforms(
    ItemToModify="((Item=(Type=Bone,Name=\"dyn_hair_r_A_Master\"),Transform=(Rotation=(X=0,Y=0,Z=0,W=1),Translation=(X=0,Y=0,Z=0),Scale3D=(X=1,Y=1,Z=1))),(Item=(Type=Bone,Name=\"dyn_hair_r_B_Master\"),Transform=(Rotation=(X=0,Y=0,Z=0,W=1),Translation=(X=0,Y=0,Z=0),Scale3D=(X=1, ...(truncated)"
    Weight=1
    ItemToModify.0.Transform.Rotation=@RigVMFunction_MathQuaternionFromEuler.Result
    ItemToModify.1.Transform.Rotation=@RigVMFunction_MathQuaternionFromEuler.Result
   )
   n12 = &RigUnit_SphericalPoseReader(
    ActiveRegionScaleFactors=(PositiveWidth=2.5,NegativeWidth=3,PositiveHeight=3,NegativeHeight=3)
    ActiveRegionSize=0.1
    Debug=(bDrawDebug=false,DebugScale=25,DebugSegments=20,DebugThickness=0.25)
    DriverAxis=(X=1,Y=0,Z=0)
    DriverItem=(Type=Bone,Name="clavicle_l")
    FalloffRegionScaleFactors=(PositiveWidth=1,NegativeWidth=1,PositiveHeight=1,NegativeHeight=1)
    FalloffSize=0.5
    OptionalParentItem=(Type=Bone,Name="None")
    RotationOffset=(X=0,Y=0,Z=0)
    ExecutePin=@SequenceAggregate.E
   )
   Multiply_5_1 = RigVMAggregateNode(A=-60, B=@n12.OutputParam, C=@n10.OutputParam)
   n13 = &RigVMFunction_MathQuaternionFromEuler(Euler=(X=0,Y=0,Z=0), Euler.Y=@Multiply_5_1.Result)
   > &RigUnit_ModifyTransforms(
    ItemToModify="((Item=(Type=Bone,Name=\"dyn_hair_l_A_Master\"),Transform=(Rotation=(X=0,Y=0,Z=0,W=1),Translation=(X=0,Y=0,Z=0),Scale3D=(X=1,Y=1,Z=1))),(Item=(Type=Bone,Name=\"dyn_hair_l_B_Master\"),Transform=(Rotation=(X=0,Y=0,Z=0,W=1),Translation=(X=0,Y=0,Z=0),Scale3D=(X=1, ...(truncated)"
    Weight=1
    ItemToModify.0.Transform.Rotation=@n13.Result
    ItemToModify.1.Transform.Rotation=@n13.Result
   )
  ]}
 SequenceAggregate_1_ContainedGraph {nodes [
   n1 = RigVMFunctionEntryNode()
   SequenceAggregate = &RigVMFunction_Sequence(ExecuteContext=@n1.ExecuteContext)
   SequenceAggregate_1 = &RigVMFunction_Sequence(ExecuteContext=@SequenceAggregate.B)
   SequenceAggregate_2 = &RigVMFunction_Sequence(ExecuteContext=@SequenceAggregate_1.B)
   SequenceAggregate_3 = &RigVMFunction_Sequence(ExecuteContext=@SequenceAggregate_2.B)
   RigVMFunctionReturnNode(A=@SequenceAggregate.A, B=@SequenceAggregate_1.A, C=@SequenceAggregate_2.A, D=@SequenceAggregate_3.A, E=@SequenceAggregate_3.B)
  ]}
]
hierarchy [
 root {loc=<0 0 0>, rot=<0 0 0 1>}
 pelvis {parent=root, loc=<0 -0.5301 98.0571>, rot=<0.0127 0.707 -0.0127 -0.707>}
 spine_01 {parent=pelvis, loc=<6.1208 -0.9328 0>, rot=<0 0 0.034 -0.9994>}
 spine_02 {parent=spine_01, loc=<6.5997 -0.2095 0>, rot=<0 0 0.0238 -0.9997>}
 spine_03 {parent=spine_02, loc=<6.4914 -0.1237 0>, rot=<0 0 0.0149 -0.9999>}
 spine_04 {parent=spine_03, loc=<6.9598 0.0794 0>, rot=<0 0 -0.0132 -0.9999>}
 spine_05 {parent=spine_04, loc=<6.9987 0 0>, rot=<0 0 -0.0015 -1>}
 neck_01 {parent=spine_05, loc=<11.3138 0.1529 0>, rot=<0 0 -0.2003 -0.9797>}
 neck_02 {parent=neck_01, loc=<4.5627 0 0>, rot=<0 0 0.0504 -0.9987>}
 head {parent=neck_02, loc=<5.3914 0 0>, rot=<0 0 0.115 -0.9934>}
 dyn_hair_l_A_Master {parent=head, loc=<4.9831 -1.4218 -7.8619>, rot=<0 -1 0 0>}
 dyn_hair_l_A_01 {parent=dyn_hair_l_A_Master, loc=<0 0 0>, rot=<0 0 0 1>}
 dyn_hair_l_A_02 {parent=dyn_hair_l_A_01, loc=<5.1518 0 0>, rot=<0 0 0 1>}
 dyn_hair_l_A_03 {parent=dyn_hair_l_A_02, loc=<5.7808 0 0>, rot=<0 0 0 1>}
 dyn_hair_l_B_Master {parent=head, loc=<3.7937 -7.006 -5.2591>, rot=<0 -0.9063 0.4226 0>}
 dyn_hair_l_B_01 {parent=dyn_hair_l_B_Master, loc=<0 0 0>, rot=<0 0 0 1>}
 dyn_hair_l_B_02 {parent=dyn_hair_l_B_01, loc=<4.5618 0 0>, rot=<0 0 0 1>}
 dyn_hair_l_B_03 {parent=dyn_hair_l_B_02, loc=<4.2466 0 0>, rot=<0 0 0 1>}
 dyn_hair_c_Master {parent=head, loc=<3.6124 -9.6163 -0.3076>, rot=<0 -0.7071 0.7071 0>}
 dyn_hair_c_01 {parent=dyn_hair_c_Master, loc=<0 0 0>, rot=<0 0 0 1>}
 dyn_hair_c_02 {parent=dyn_hair_c_01, loc=<4.5618 0 0>, rot=<0 0 0 1>}
 dyn_hair_c_03 {parent=dyn_hair_c_02, loc=<4.2466 0 0>, rot=<0 0 0 1>}
 dyn_hair_r_A_Master {parent=head, loc=<4.9827 -1.4218 7.8619>, rot=<0 0 1 0>}
 dyn_hair_r_A_01 {parent=dyn_hair_r_A_Master, loc=<0 0 0>, rot=<0 0 0 -1>}
 dyn_hair_r_A_02 {parent=dyn_hair_r_A_01, loc=<5.152 0 0>, rot=<0 0 0 -1>}
 dyn_hair_r_A_03 {parent=dyn_hair_r_A_02, loc=<5.781 0 0>, rot=<0 0 0 -1>}
 dyn_hair_r_B_Master {parent=head, loc=<3.7938 -7.006 5.2591>, rot=<0 -0.2588 0.9659 0>}
 dyn_hair_r_B_01 {parent=dyn_hair_r_B_Master, loc=<0 0 0>, rot=<0 0 0 -1>}
 dyn_hair_r_B_02 {parent=dyn_hair_r_B_01, loc=<4.562 0 0>, rot=<0 0 0 -1>}
 dyn_hair_r_B_03 {parent=dyn_hair_r_B_02, loc=<4.247 0 0>, rot=<0 0 0 -1>}
 dyn_side_hair_r {parent=head, loc=<5.4861 2.9996 9.6679>, rot=<0.0041 -0.9999 -0.0001 -0.0153>}
 dyn_side_hair_l {parent=head, loc=<3.9962 2.4349 -9.8195>, rot=<0.0041 -0.999 0.0002 0.045>}
 hair_base {parent=head, loc=<9.5167 -9.4141 0.5875>, rot=<0.0041 -0.9999 -0.0001 -0.0131>}
 base_back_hair_01 {parent=hair_base, loc=<4.9431 5.4731 -6.8247>, rot=<0.2736 0.0245 -0.0023 -0.9615>}
 dyn_back_hair_01 {parent=base_back_hair_01, loc=<0 0 0>, rot=<0 0 0 1>}
 base_back_hair_02 {parent=hair_base, loc=<5.028 1.2303 -5.0884>, rot=<-0.3842 -0.0249 0.0108 -0.9229>}
 dyn_back_hair_02 {parent=base_back_hair_02, loc=<0 0 0>, rot=<0 0 0 1>}
 base_back_hair_03 {parent=hair_base, loc=<5.1504 0.5125 -1.3715>, rot=<0 0.0276 0 -0.9996>}
 dyn_back_hair_03 {parent=base_back_hair_03, loc=<0 0 0>, rot=<0 0 0 1>}
 base_back_hair_04 {parent=hair_base, loc=<5.4075 0.345 4.0933>, rot=<0.1701 -0.0038 0.0026 -0.9854>}
 dyn_back_hair_04 {parent=base_back_hair_04, loc=<0 0 0>, rot=<0 0 0 1>}
 base_back_hair_05 {parent=hair_base, loc=<5.2723 3.2074 7.1139>, rot=<0.4243 0.0048 0.0327 -0.9049>}
 dyn_back_hair_05 {parent=base_back_hair_05, loc=<0 0 0>, rot=<0 0 0 1>}
 base_back_hair_06 {parent=hair_base, loc=<5.3978 5.4397 8.8094>, rot=<-0.1767 0.0358 -0.0227 -0.9833>}
 dyn_back_hair_06 {parent=base_back_hair_06, loc=<0 0 0>, rot=<0 0 0 1>}
 base_back_hair_07 {parent=hair_base, loc=<5.2144 -0.4988 0.784>, rot=<0 0.0276 0 -0.9996>}
 dyn_back_hair_07 {parent=base_back_hair_07, loc=<0 0 0>, rot=<0 0 0 1>}
 base_back_hair_08 {parent=hair_base, loc=<5.3498 2.3676 8.0048>, rot=<0.4243 0.0048 0.0327 -0.9049>}
 dyn_back_hair_08 {parent=base_back_hair_08, loc=<0 0 0>, rot=<0 0 0 1>}
 base_back_hair_09 {parent=hair_base, loc=<5.2186 0.3533 3.4228>, rot=<0 0.0276 0 -0.9996>}
 dyn_back_hair_09 {parent=base_back_hair_09, loc=<0 0 0>, rot=<0 0 0 1>}
 base_back_hair_10 {parent=hair_base, loc=<4.8682 5.9027 -8.8413>, rot=<0.0161 0.0231 -0.0085 -0.9996>}
 dyn_back_hair_10 {parent=base_back_hair_10, loc=<0 0 0>, rot=<0 0 0 1>}
 base_back_hair_11 {parent=hair_base, loc=<5.0043 3.7112 -7.3458>, rot=<0.2697 0.0245 -0.0024 -0.9626>}
 dyn_back_hair_11 {parent=base_back_hair_11, loc=<0 0 0>, rot=<0 0 0 1>}
 dyn_top_hair_01 {parent=head, loc=<14.5326 -6.517 0.8136>, rot=<0.4443 0.4157 -0.1728 -0.7745>}
 dyn_top_hair_02 {parent=head, loc=<14.2699 -7.4324 1.5584>, rot=<0.0768 0.608 -0.1099 -0.7825>}
 dyn_top_hair_03 {parent=head, loc=<14.3713 -8.02 -0.0922>, rot=<-0.1093 0.0771 -0.0402 -0.9902>}
 dyn_top_hair_04 {parent=head, loc=<13.3391 -8.7759 -1.0488>, rot=<0.1461 -0.3187 0.0456 -0.9354>}
 dyn_front_mask_hair {parent=head, loc=<14.0184 13.8885 1.359>, rot=<0.004 -0.9998 -0.0207 0.0055>}
 faceAttach {parent=head, loc=<-0.3332 -1.8708 0.1637>, rot=<0.0029 -0.7071 -0.0029 -0.7071>}
 C_brow_mid {parent=faceAttach, loc=<0.1686 11.7505 4.6079>, rot=<0 0 0 -1>}
 L_brow_mid {parent=faceAttach, loc=<4.1325 11.0282 5.9226>, rot=<-0.0001 -0.0042 0.2553 -0.9669>}
 L_brow_outer {parent=faceAttach, loc=<7.8239 8.761 5.7545>, rot=<-0.0287 -0.04 0.4491 -0.8921>}
 L_cheek_inner {parent=faceAttach, loc=<5.8133 9.9332 -0.7202>, rot=<-0.0706 -0.2697 0.3184 -0.906>}
 C_lip_upper_mid {parent=faceAttach, loc=<0.1692 12.3772 -3.1537>, rot=<0.2225 0.0013 -0.0005 -0.9749>}
 L_lip_upper_outer {parent=faceAttach, loc=<1.7851 11.7499 -3.2461>, rot=<0.0359 -0.0728 0.2627 -0.9615>}
 L_lip_corner {parent=faceAttach, loc=<3.1946 11.3833 -3.3728>, rot=<0.003 0.008 -0.0146 -0.9999>}
 C_nose_bridge {parent=faceAttach, loc=<0.1579 12.4643 0.8146>, rot=<-0.2218 0 0 -0.9751>}
 R_brow_mid {parent=faceAttach, loc=<-3.7169 11.0561 5.8778>, rot=<-0.0001 0.0042 -0.2553 -0.9669>}
 R_brow_outer {parent=faceAttach, loc=<-7.3157 9.0743 5.9312>, rot=<-0.0287 0.04 -0.4491 -0.8921>}
 R_cheek_inner {parent=faceAttach, loc=<-5.4333 9.9544 -0.694>, rot=<-0.0706 0.2697 -0.3184 -0.906>}
 R_lip_corner {parent=faceAttach, loc=<-2.7722 11.3443 -3.3317>, rot=<0.003 -0.008 0.0146 -0.9999>}
 C_jaw {parent=faceAttach, loc=<0.079 1.3131 -0.8587>, rot=<0 0 0 1>}
 L_lip_lower_outer {parent=C_jaw, loc=<1.7074 10.3157 -2.3662>, rot=<0.013 0.0445 0.3057 -0.951>}
 C_lip_lower_mid {parent=C_jaw, loc=<0.0725 10.8276 -2.3215>, rot=<0 0 0 -1>}
 R_lip_lower_outer {parent=C_jaw, loc=<-1.742 10.2965 -2.2757>, rot=<0.013 -0.0445 -0.3057 -0.951>}
 teeth_lower {parent=C_jaw, loc=<0.0581 9.393 -3.3464>, rot=<0 0 0 1>}
 tongue {parent=C_jaw, loc=<0.2084 4.064 -1.0402>, rot=<-0.1568 0 0 -0.9876>}
 R_eye {parent=faceAttach, loc=<-3.9027 8.2576 3.3157>, rot=<0 0 0.9998 -0.0175>}
 R_eye_lid_upper_mid {parent=R_eye, loc=<0 0 0>, rot=<0.0523 0 0 -0.9986>}
 R_eye_lid_lower_mid {parent=R_eye, loc=<0 0 0>, rot=<0.0329 0 0 -0.9995>}
 L_eye {parent=faceAttach, loc=<4.2302 8.2564 3.3153>, rot=<0 0 0.9998 0.0175>}
 L_eye_lid_upper_mid {parent=L_eye, loc=<0 0 0>, rot=<0.0523 0 0 -0.9986>}
 L_eye_lid_lower_mid {parent=L_eye, loc=<0 0 0>, rot=<0.0329 0 0 -0.9995>}
 teeth_upper {parent=faceAttach, loc=<0.1533 10.0797 -1.7485>, rot=<0 0 0 1>}
 R_lip_upper_outer {parent=faceAttach, loc=<-1.4581 11.7467 -3.2469>, rot=<0.0359 0.0728 -0.2627 -0.9615>}
 dyn_front_01 {parent=head, loc=<6.105 12.5495 -3.6262>, rot=<0.0041 -0.9857 0.1685 -0.0007>}
 dyn_front_02 {parent=head, loc=<6.1838 10.2858 -7.0171>, rot=<0.0038 -0.9274 0.3741 -0.0015>}
 dyn_front_03 {parent=head, loc=<5.9901 7.0009 -9.2367>, rot=<0.0035 -0.8468 0.5319 -0.0022>}
 dyn_front_04 {parent=head, loc=<6.0603 12.6411 1.2584>, rot=<0.0041 -0.9982 -0.0601 0.0002>}
 dyn_front_05 {parent=head, loc=<5.953 11.1059 6.4951>, rot=<0.004 -0.9667 -0.256 0.0011>}
 dyn_front_06 {parent=head, loc=<6.4548 7.5406 9.5177>, rot=<0.0036 -0.8646 -0.5024 0.0021>}
 clavicle_l {parent=spine_05, loc=<5.4831 -0.1053 -3.2946>, rot=<0.0465 -0.7289 -0.009 -0.6829>}
 upperarm_l {parent=clavicle_l, loc=<9.8658 -1.2136 -0.4105>, rot=<0.0164 -0.3245 -0.0269 -0.9454>}
 clavicle_r {parent=spine_05, loc=<5.4831 -0.1053 3.2946>, rot=<0.7289 0.0465 -0.6829 0.009>}
 upperarm_r {parent=clavicle_r, loc=<-9.8658 1.2136 0.4105>, rot=<0.0164 -0.3245 -0.0269 -0.9454>}
]
